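; Collision proxy actor for the held Shielder shield, forwarding damage it receives to the main shield and never dying itself.
; On server damage, it calls HandleDamage on the main shield reference and plays a damage ripple using the hit location, normal and reflected trace direction.
; InitFromParent links it and its collision helper component to the parent shield.
; EnableCollision toggles query-only mesh collision and helper ticking, with a debug print.
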(event EventBeginPlay
  (bind _asplayer_pawn_athena (Utilities|Casting|CastToPlayerPawn_Athena (Game|GetInstigator)))
  (Variables|Default|SetPlayerPawn _asplayer_pawn_athena))

(event EventOnDamageServer
  (params (Damage :type Float%20%28single-precision%29) (DamageTags :type Gameplay%20Tag%20Container%20Structure%20%28by%20ref%29) (Momentum :type Vector :default "0, 0, 0") (HitInfo :type Hit%20Result%20Structure%20%28by%20ref%29) (InstigatedBy :type Controller%20Object%20Reference) (DamageCauser :type Actor%20Object%20Reference) (EffectContext :type Gameplay%20Effect%20Context%20Handle%20Structure))
  (bind (_bblockinghit _binitialoverlap _time _distance _location _impactpoint _normal _impactnormal _physmat _hitactor _hitcomponent _hitbonename _bonename _hititem _elementindex _faceindex _tracestart _traceend) (Collision|BreakHitResult HitInfo))
  (bind _mainshieldref (Variables|Default|GetMainShieldRef))
  (Utilities|IsValid _mainshieldref
    (:Is%20Valid
      (Damage|HandleDamage _mainshieldref Damage HitInfo DamageCauser)
      (Class|BGAShielderParent|PlayDamageRipple _mainshieldref Damage _location _normal (Math|Vector|GetReflectionVector (Math|Vector|Normalize (- _traceend _tracestart)) _normal)))))

(fn EnableCollision
  (params (bEnabled :type Boolean))
  (Development|PrintString (Utilities|String|Append "Shield collision: " (Utilities|String|ToString(Boolean) bEnabled)) true true "(R=1.000000,G=0.000000,B=0.365363,A=1.000000)")
  (Collision|SetCollisionEnabled (Variables|Default|GetStaticMesh) (select bEnabled "QueryOnly" "NoCollision"))
  (Components|Tick|SetComponentTickEnabled (Variables|Default|GetShielderCollisionHelper) bEnabled))

(fn InitFromParent
  (params (Parent :type BGA%20Shielder%20Held%20Object%20Reference))
  (bind _output_get (Variables|Default|SetMainShieldRef Parent))
  (Variables|/Script/ShielderGameplayRuntime.ShielderCollisionHelperComponent|SetShielderBGA _output_get (Variables|Default|GetShielderCollisionHelper)))

(fn ShouldDie
  (params (Damage :type Float%20%28single-precision%29) (EventInstigator :type Controller%20Object%20Reference) (DamageCauser :type Actor%20Object%20Reference) (bIsTest :type Boolean) (EffectContext :type Gameplay%20Effect%20Context%20Handle%20Structure%20%28by%20ref%29))
  (return false))

(fn ConstructionScript)
